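# Physics asset for the dynamic necklace and arm feather bones of the female medium Islander player cosmetic body.
# Three necklace capsules form a chain from spine_05, with twist locked and swing limit angles set on each link.
# Each side's feather master body is rotation-locked to spine_05, with two feathers branching from it under tight angular limits.
# Three larger capsules on spine_05 form the torso volume.
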PhysicsAsset /BRCosmetics/Characters/Player/Female/Medium/Bodies/F_MED_Islander/Meshes/F_MED_Islander_dyn_necklace
bodies {
 dyn_arm_feather_l_01 [{shapeType=Capsule, length=6.75, radius=1, center=<2 0.0001 0>, rotation=<rot 90 0 0>}]
 dyn_arm_feather_l_02 [{shapeType=Capsule, length=6.75, radius=1, center=<2 0.0001 0>, rotation=<rot 90 0 0>}]
 dyn_arm_feather_l_MSTR [{shapeType=Capsule, length=13.767, radius=3.9418, center=<-2.3091 0 -3.1409>, rotation=<rot -44.9996 -180 180>}]
 dyn_arm_feather_r_01 [{shapeType=Capsule, length=6.75, radius=1, center=<2 0.0001 0>, rotation=<rot 90 0 0>}]
 dyn_arm_feather_r_02 [{shapeType=Capsule, length=6.75, radius=1, center=<2 0.0001 0>, rotation=<rot 90 0 0>}]
 dyn_arm_feather_r_MSTR [{shapeType=Capsule, length=13.767, radius=3.9418, center=<3.5799 -0.7404 -2.1666>, rotation=<rot -44.9995 0 0>}]
 dyn_necklace_01 [{shapeType=Capsule, length=4.3015, radius=1.1843, shapeName=dyn_necklace_01_capsule, center=<2.4234 0 0>, rotation=<rot 0 90 90>}]
 dyn_necklace_02 [{shapeType=Capsule, length=4.3945, radius=1.1843, shapeName=dyn_necklace_02_capsule, center=<2.3729 0 0>, rotation=<rot 0 90 90>}]
 dyn_necklace_03 [{shapeType=Capsule, length=3, radius=0.8, center=<2 0 0>, rotation=<rot 90 0 0>}]
 spine_05 [
  {shapeType=Capsule, length=4.5489, radius=9, center=<-3.9197 1.9159 0.1478>, rotation=<rot -0.0007 50.0002 177.9725>}
  {shapeType=Capsule, length=9.3128, radius=6.7401, shapeName=spine_05_capsule, center=<-5.8226 6.1258 -0.4164>}
  {shapeType=Capsule, length=5.1686, radius=8.0589, shapeName=spine_05_capsule}
 ]
}
constraints [
 {bone1Name=dyn_necklace_01, bone2Name=spine_05, swing1LimitDegrees=20, swing2LimitDegrees=30, twistLimitDegrees=5, twistMotion=Locked}
 {bone1Name=dyn_necklace_02, bone2Name=dyn_necklace_01, swing1LimitDegrees=20, swing2LimitDegrees=30, twistLimitDegrees=5, twistMotion=Locked}
 {bone1Name=dyn_necklace_03, bone2Name=dyn_necklace_02, swing1LimitDegrees=20, swing2LimitDegrees=30, twistLimitDegrees=5, twistMotion=Locked}
 {
  bone1Name=dyn_arm_feather_l_MSTR
  bone2Name=spine_05
  swing1LimitDegrees=45
  swing1Motion=Locked
  swing2LimitDegrees=45
  swing2Motion=Locked
  twistLimitDegrees=45
  twistMotion=Locked
 }
 {bone1Name=dyn_arm_feather_l_01, bone2Name=dyn_arm_feather_l_MSTR, swing1LimitDegrees=5, swing2LimitDegrees=5, twistLimitDegrees=1}
 {bone1Name=dyn_arm_feather_l_02, bone2Name=dyn_arm_feather_l_MSTR, swing1LimitDegrees=5, swing2LimitDegrees=5, twistLimitDegrees=1}
 {bone1Name=dyn_arm_feather_r_01, bone2Name=dyn_arm_feather_r_MSTR, swing1LimitDegrees=5, swing2LimitDegrees=5, twistLimitDegrees=1}
 {
  bone1Name=dyn_arm_feather_r_MSTR
  bone2Name=spine_05
  swing1LimitDegrees=45
  swing1Motion=Locked
  swing2LimitDegrees=45
  swing2Motion=Locked
  twistLimitDegrees=45
  twistMotion=Locked
 }
 {bone1Name=dyn_arm_feather_r_02, bone2Name=dyn_arm_feather_r_MSTR, swing1LimitDegrees=5, swing2LimitDegrees=5, twistLimitDegrees=1}
]
