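# Physics asset for the female medium Islander body's dynamic skirt, leaves, pompon and bag, apparently a performance variant.
# Simple capsules and boxes cover those bone chains plus pelvis, lower spine, thighs and calves.
# Five skirt chains hang from thigh twist bones through free-rotating MSTR bodies, with tight, twist-locked swing limits below.
# Leaves and bag anchor to the pelvis (dyn_leaves_bk_03 directly, not to bk_02); the pompon hangs from spine_01.
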PhysicsAsset /BRCosmetics/Characters/Player/Female/Medium/Bodies/F_MED_Islander/Meshes/F_MED_Islander_dyn_skirt_perf
bodies {
 calf_l [{shapeType=Capsule, length=44.25, radius=6, center=<-22.634 -4.1024 -0.3>, rotation=<rot 90 0 -6>}]
 calf_r [{shapeType=Capsule, length=44.25, radius=6, center=<23.6699 4.2113 0.3>, rotation=<rot 90 0 -6>}]
 dyn_bag_02 [{
   shapeType=Box
   extentX=15
   extentY=4.5076
   extentZ=15.2484
   shapeName=dyn_bag_02_box
   center=<7.7841 0.8383 2.0037>
   rotation=<rot 0 0 89.9999>
  }]
 dyn_bag_03 [{shapeType=Capsule, length=12, radius=1.5, shapeName=dyn_bag_03_capsule, center=<3.5 0 0.5142>, rotation=<rot 0 0 89.9999>}]
 dyn_bag_MSTR [{shapeType=Capsule, length=1, radius=1}]
 dyn_bag_hanging_01 [{shapeType=Capsule, length=1, radius=0.6, center=<3 0 0>, rotation=<rot 90 0 0>}]
 dyn_bag_hanging_02 [{shapeType=Capsule, length=2, radius=1, center=<2 0 0>, rotation=<rot 90 0 0>}]
 dyn_leaves_bk_01 [{shapeType=Capsule, length=4, radius=2, shapeName=dyn_leaves_ft_02_capsule, center=<4 0 0>, rotation=<rot 90 0 0>}]
 dyn_leaves_bk_02 [{shapeType=Capsule, length=10.25, radius=2, shapeName=dyn_leaves_ft_03_capsule, center=<10.1388 0.0002 1.0122>, rotation=<rot 90 0 0>}]
 dyn_leaves_bk_03 [{shapeType=Capsule, length=4, radius=2, shapeName=dyn_leaves_ft_03_capsule, center=<5 -0.0001 1.0122>, rotation=<rot 90 0.0197 0.0195>}]
 dyn_leaves_bk_l_01 [{shapeType=Capsule, length=4, radius=2, shapeName=dyn_leaves_ft_02_capsule, center=<4 0 0>, rotation=<rot 90 0 0>}]
 dyn_leaves_bk_l_02 [{shapeType=Capsule, length=5, radius=2, shapeName=dyn_leaves_ft_03_capsule, center=<5 -0.0001 1.0122>, rotation=<rot 90 0 0>}]
 dyn_leaves_bk_r_01 [{shapeType=Capsule, length=8, radius=2, shapeName=dyn_leaves_ft_03_capsule, center=<6 0 0>, rotation=<rot 90 0 0>}]
 dyn_leaves_ft_02 [{shapeType=Capsule, length=23.25, radius=2.9577, shapeName=dyn_leaves_ft_02_capsule, center=<9.44 10 1.7006>, rotation=<rot 0 0 89.9999>}]
 dyn_leaves_ft_03 [{shapeType=Capsule, length=6.25, radius=2.9873, shapeName=dyn_leaves_ft_03_capsule, center=<9 10 1.7006>, rotation=<rot 90 0 0>}]
 dyn_leaves_ft_l_02 [{shapeType=Capsule, length=5.5, radius=2.5, shapeName=dyn_leaves_ft_02_capsule, center=<5 0.0001 2.468>, rotation=<rot 0 0 89.9999>}]
 dyn_leaves_ft_l_03 [{shapeType=Capsule, length=3.9612, radius=2.0388, shapeName=dyn_leaves_ft_03_capsule, center=<6 -0.0001 2.1924>, rotation=<rot 0 0 90>}]
 dyn_leaves_l_02 [{shapeType=Capsule, length=5.25, radius=2.75, shapeName=dyn_leaves_ft_02_capsule, center=<5 0 0>, rotation=<rot 90 0 0>}]
 dyn_leaves_l_03 [{shapeType=Capsule, length=1.25, radius=2.75, shapeName=dyn_leaves_ft_03_capsule, center=<6 -0.0001 0.2841>, rotation=<rot 90 0 0>}]
 dyn_leaves_l_bk_01 [{shapeType=Capsule, length=4, radius=2, shapeName=dyn_leaves_ft_02_capsule, center=<4 0.0001 -0.6177>, rotation=<rot 90 0 0>}]
 dyn_leaves_l_bk_02 [{shapeType=Capsule, length=4, radius=2, shapeName=dyn_leaves_ft_03_capsule, center=<5 0 -0.6248>, rotation=<rot 90 0 0>}]
 dyn_pompon_01 [{shapeType=Capsule, length=2, radius=1, center=<3.7566 -0.0001 0>, rotation=<rot 90 0 0>}]
 dyn_pompon_02 [{shapeType=Capsule, length=4.25, radius=1.5, center=<5.7385 -0.0001 0.1721>, rotation=<rot 90 0 0>}]
 dyn_pompon_03 [{shapeType=Capsule, length=2.75, radius=2.2, center=<4.7608 0 0>, rotation=<rot 90 0 0>}]
 dyn_skirt_bk_l_01 [{shapeType=Capsule, length=20, radius=3, shapeName=dyn_skirt_bk_l_01_capsule, center=<8 0 1>, rotation=<rot 0 0 90>}]
 dyn_skirt_bk_l_02 [{shapeType=Capsule, length=20, radius=3, shapeName=dyn_skirt_bk_l_02_capsule, center=<6 0 2>, rotation=<rot 0 0 90>}]
 dyn_skirt_bk_l_MSTR [{shapeType=Capsule, length=1, radius=1, rotation=<rot 90 0 0>}]
 dyn_skirt_bk_r_01 [{shapeType=Capsule, length=20, radius=3, shapeName=dyn_skirt_bk_r_01_capsule, center=<8 0 2>, rotation=<rot 0 0 79.9998>}]
 dyn_skirt_bk_r_02 [{shapeType=Capsule, length=20, radius=3, shapeName=dyn_skirt_bk_r_02_capsule, center=<6 0 2>, rotation=<rot 0 0 79.9998>}]
 dyn_skirt_bk_r_MSTR [{shapeType=Capsule, length=1, radius=1, rotation=<rot 90 0 0>}]
 dyn_skirt_ft_l_01 [{shapeType=Box, extentX=20, extentY=20, extentZ=3, shapeName=dyn_skirt_ft_l_01_box, center=<1 0 0>, rotation=<rot 0 0 -10.0007>}]
 dyn_skirt_ft_l_02 [{shapeType=Capsule, length=20, radius=3, shapeName=dyn_skirt_ft_l_02_capsule, center=<6 0 2>, rotation=<rot 0 0 79.9993>}]
 dyn_skirt_ft_l_MSTR [{shapeType=Capsule, length=1, radius=1, rotation=<rot 90 0 0>}]
 dyn_skirt_ft_r_01 [{shapeType=Box, extentX=20, extentY=24, extentZ=3, shapeName=dyn_skirt_ft_r_01_box, center=<1.075 -2 1.3294>, rotation=<rot -5 0 0>}]
 dyn_skirt_ft_r_02 [{shapeType=Capsule, length=20, radius=3, shapeName=dyn_skirt_ft_r_02_capsule, center=<6 0 2>, rotation=<rot 0 0 90>}]
 dyn_skirt_ft_r_MSTR [{shapeType=Capsule, length=1, radius=1, rotation=<rot 90 0 0>}]
 dyn_skirt_l_01 [{shapeType=Capsule, length=15, radius=3, shapeName=dyn_skirt_l_01_capsule, center=<8 0 1>, rotation=<rot 0 0 90>}]
 dyn_skirt_l_02 [{shapeType=Capsule, length=15, radius=3, shapeName=dyn_skirt_l_02_capsule, center=<6 0 2>, rotation=<rot 0 0 90>}]
 dyn_skirt_l_MSTR [{shapeType=Capsule, length=1, radius=1, rotation=<rot 90 0 0>}]
 pelvis [{shapeType=Capsule, length=16.7077, radius=9.998, shapeName=pelvis_capsule, center=<-9.7078 0.0002 0>}]
 spine_01 [
  {shapeType=Capsule, length=5.75, radius=9.5, center=<-1.8977 -0.6959 0.0013>, rotation=<rot 0 0 0.0035>}
  {shapeType=Capsule, length=22, radius=8, shapeName=spine_01_capsule, center=<7.3746 3.1318 0>, rotation=<rot -90 -0.0297 0.0297>}
 ]
 thigh_twist_01_l [{shapeType=Capsule, length=18.75, radius=7.75, center=<3.548 -1.6301 -0.0882>, rotation=<rot 90 0 -0.0442>}]
 thigh_twist_01_r [
  {shapeType=Capsule, length=27.5, radius=7, center=<-3.6406 -0.5457 -0.5569>, rotation=<rot 84.996 179.0581 -0.4727>}
  {
   shapeType=Capsule
   length=27.5
   radius=7
   shapeName=thigh_twist_01_r_capsule
   center=<-3.8234 -0.4701 -11.8169>
   rotation=<rot 84.9923 179.0581 -0.4726>
  }
 ]
}
constraints [
 {bone1Name=thigh_twist_01_l, bone2Name=pelvis, swing1LimitDegrees=45, swing2LimitDegrees=45, twistLimitDegrees=45}
 {bone1Name=thigh_twist_01_r, bone2Name=pelvis, swing1LimitDegrees=45, swing2LimitDegrees=45, twistLimitDegrees=45}
 {
  bone1Name=dyn_skirt_bk_l_MSTR
  bone2Name=thigh_twist_01_l
  swing1LimitDegrees=45
  swing1Motion=Free
  swing2LimitDegrees=45
  swing2Motion=Free
  twistLimitDegrees=45
  twistMotion=Free
 }
 {
  bone1Name=dyn_skirt_bk_l_01
  bone2Name=dyn_skirt_bk_l_MSTR
  swing1LimitDegrees=3
  swing2LimitDegrees=25
  twistLimitDegrees=2
  twistMotion=Locked
 }
 {
  bone1Name=dyn_skirt_bk_l_02
  bone2Name=dyn_skirt_bk_l_01
  swing1LimitDegrees=3
  swing2LimitDegrees=25
  twistLimitDegrees=2
  twistMotion=Locked
 }
 {
  bone1Name=dyn_skirt_ft_l_MSTR
  bone2Name=thigh_twist_01_l
  swing1LimitDegrees=45
  swing1Motion=Free
  swing2LimitDegrees=45
  swing2Motion=Free
  twistLimitDegrees=45
  twistMotion=Free
 }
 {
  bone1Name=dyn_skirt_ft_l_01
  bone2Name=dyn_skirt_ft_l_MSTR
  swing1LimitDegrees=3
  swing2LimitDegrees=15
  twistLimitDegrees=2
  twistMotion=Locked
 }
 {
  bone1Name=dyn_skirt_ft_l_02
  bone2Name=dyn_skirt_ft_l_01
  swing1LimitDegrees=3
  swing2LimitDegrees=15
  twistLimitDegrees=2
  twistMotion=Locked
 }
 {
  bone1Name=dyn_skirt_l_MSTR
  bone2Name=thigh_twist_01_l
  swing1LimitDegrees=45
  swing1Motion=Free
  swing2LimitDegrees=45
  swing2Motion=Free
  twistLimitDegrees=45
  twistMotion=Free
 }
 {bone1Name=dyn_skirt_l_01, bone2Name=dyn_skirt_l_MSTR, swing1LimitDegrees=3, swing2LimitDegrees=15, twistLimitDegrees=2, twistMotion=Locked}
 {bone1Name=dyn_skirt_l_02, bone2Name=dyn_skirt_l_01, swing1LimitDegrees=3, swing2LimitDegrees=15, twistLimitDegrees=2, twistMotion=Locked}
 {
  bone1Name=dyn_skirt_ft_r_MSTR
  bone2Name=thigh_twist_01_r
  swing1LimitDegrees=45
  swing1Motion=Free
  swing2LimitDegrees=45
  swing2Motion=Free
  twistLimitDegrees=45
  twistMotion=Free
 }
 {
  bone1Name=dyn_skirt_ft_r_01
  bone2Name=dyn_skirt_ft_r_MSTR
  swing1LimitDegrees=3
  swing2LimitDegrees=20
  twistLimitDegrees=2
  twistMotion=Locked
 }
 {
  bone1Name=dyn_skirt_ft_r_02
  bone2Name=dyn_skirt_ft_r_01
  swing1LimitDegrees=3
  swing2LimitDegrees=25
  twistLimitDegrees=2
  twistMotion=Locked
 }
 {
  bone1Name=dyn_skirt_bk_r_MSTR
  bone2Name=thigh_twist_01_r
  swing1LimitDegrees=45
  swing1Motion=Free
  swing2LimitDegrees=45
  swing2Motion=Free
  twistLimitDegrees=45
  twistMotion=Free
 }
 {
  bone1Name=dyn_skirt_bk_r_01
  bone2Name=dyn_skirt_bk_r_MSTR
  swing1LimitDegrees=3
  swing2LimitDegrees=20
  twistLimitDegrees=2
  twistMotion=Locked
 }
 {
  bone1Name=dyn_skirt_bk_r_02
  bone2Name=dyn_skirt_bk_r_01
  swing1LimitDegrees=3
  swing2LimitDegrees=25
  twistLimitDegrees=2
  twistMotion=Locked
 }
 {
  bone1Name=calf_r
  bone2Name=pelvis
  swing1LimitDegrees=45
  swing1Motion=Free
  swing2LimitDegrees=45
  swing2Motion=Free
  twistLimitDegrees=45
  twistMotion=Free
 }
 {
  bone1Name=calf_l
  bone2Name=pelvis
  swing1LimitDegrees=45
  swing1Motion=Free
  swing2LimitDegrees=45
  swing2Motion=Free
  twistLimitDegrees=45
  twistMotion=Free
 }
 {
  bone1Name=spine_01
  bone2Name=pelvis
  swing1LimitDegrees=45
  swing1Motion=Free
  swing2LimitDegrees=45
  swing2Motion=Free
  twistLimitDegrees=45
  twistMotion=Free
 }
 {bone1Name=dyn_pompon_01, bone2Name=spine_01, swing1LimitDegrees=20, swing2LimitDegrees=10, twistLimitDegrees=5}
 {bone1Name=dyn_pompon_02, bone2Name=dyn_pompon_01, swing1LimitDegrees=20, swing2LimitDegrees=30, twistLimitDegrees=5}
 {bone1Name=dyn_pompon_03, bone2Name=dyn_pompon_02, swing1LimitDegrees=20, swing2LimitDegrees=30, twistLimitDegrees=5}
 {
  bone1Name=dyn_bag_MSTR
  bone2Name=pelvis
  swing1LimitDegrees=45
  swing1Motion=Free
  swing2LimitDegrees=45
  swing2Motion=Free
  twistLimitDegrees=45
  twistMotion=Free
 }
 {bone1Name=dyn_bag_02, bone2Name=dyn_bag_MSTR, swing1LimitDegrees=1, swing2LimitDegrees=20, twistLimitDegrees=1}
 {bone1Name=dyn_bag_03, bone2Name=dyn_bag_02, swing1LimitDegrees=1, swing2LimitDegrees=45, twistLimitDegrees=1, twistMotion=Locked}
 {bone1Name=dyn_bag_hanging_01, bone2Name=dyn_bag_02, swing1LimitDegrees=2, swing2LimitDegrees=20, twistLimitDegrees=1, twistMotion=Locked}
 {
  bone1Name=dyn_bag_hanging_02
  bone2Name=dyn_bag_hanging_01
  swing1LimitDegrees=2
  swing2LimitDegrees=20
  twistLimitDegrees=1
  twistMotion=Locked
 }
 {bone1Name=dyn_leaves_bk_r_01, bone2Name=pelvis, swing1LimitDegrees=30, swing2LimitDegrees=25, twistLimitDegrees=1, twistMotion=Locked}
 {bone1Name=dyn_leaves_ft_02, bone2Name=pelvis, swing1LimitDegrees=5, swing2LimitDegrees=35, twistLimitDegrees=10}
 {bone1Name=dyn_leaves_ft_03, bone2Name=dyn_leaves_ft_02, swing1LimitDegrees=5, swing2LimitDegrees=35, twistLimitDegrees=1, twistMotion=Locked}
 {bone1Name=dyn_leaves_ft_l_02, bone2Name=pelvis, swing1LimitDegrees=16, swing2LimitDegrees=25, twistLimitDegrees=10}
 {bone1Name=dyn_leaves_ft_l_03, bone2Name=dyn_leaves_ft_l_02, swing1LimitDegrees=16, swing2LimitDegrees=25, twistLimitDegrees=10}
 {bone1Name=dyn_leaves_l_02, bone2Name=pelvis, swing1LimitDegrees=20, swing2LimitDegrees=20, twistLimitDegrees=1, twistMotion=Locked}
 {bone1Name=dyn_leaves_l_03, bone2Name=dyn_leaves_l_02, swing1LimitDegrees=20, swing2LimitDegrees=20, twistLimitDegrees=1, twistMotion=Locked}
 {bone1Name=dyn_leaves_l_bk_01, bone2Name=pelvis, swing1LimitDegrees=20, swing2LimitDegrees=20, twistLimitDegrees=1, twistMotion=Locked}
 {
  bone1Name=dyn_leaves_l_bk_02
  bone2Name=dyn_leaves_l_bk_01
  swing1LimitDegrees=20
  swing2LimitDegrees=20
  twistLimitDegrees=1
  twistMotion=Locked
 }
 {bone1Name=dyn_leaves_bk_l_01, bone2Name=pelvis, swing1LimitDegrees=20, swing2LimitDegrees=20, twistLimitDegrees=1, twistMotion=Locked}
 {
  bone1Name=dyn_leaves_bk_l_02
  bone2Name=dyn_leaves_bk_l_01
  swing1LimitDegrees=20
  swing2LimitDegrees=20
  twistLimitDegrees=1
  twistMotion=Locked
 }
 {bone1Name=dyn_leaves_bk_01, bone2Name=pelvis, swing1LimitDegrees=20, swing2LimitDegrees=20, twistLimitDegrees=1, twistMotion=Locked}
 {
  bone1Name=dyn_leaves_bk_02
  bone2Name=dyn_leaves_bk_01
  swing1LimitDegrees=20
  swing2LimitDegrees=20
  twistLimitDegrees=1
  twistMotion=Locked
 }
 {bone1Name=dyn_leaves_bk_03, bone2Name=pelvis, swing1LimitDegrees=20, swing2LimitDegrees=20, twistLimitDegrees=1, twistMotion=Locked}
]
