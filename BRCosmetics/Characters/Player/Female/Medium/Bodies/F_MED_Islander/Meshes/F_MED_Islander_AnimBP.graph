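# Secondary-motion AnimBP for the female medium Islander body cosmetic, animating its skirt, leaves, bag, necklace and feathers.
# It caches the parent mesh pose after weighted bone constraints and thigh/knee-driven correctives.
# Standing and crouching skirt RigidBody branches blend by bool, then feed necklace RigidBody and feather AnimDynamics (feathers_01 runs twice, 04 never).
# Skydiving start/end and delayed parachute opening reset dynamics.
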AnimBlueprint /BRCosmetics/Characters/Player/Female/Medium/Bodies/F_MED_Islander/Meshes/F_MED_Islander_AnimBP
target_skeleton=&/F_MED_Islander_Skeleton
graphs [
 AnimGraph {nodes [
   n1 = "Transform (Modify) Bone - Bone: dyn_bag_MSTR"(
    ComponentPose=Local To Component(LocalPose=Use cached pose 'SavedPose20243'())
    Alpha=1
    Translation="0,0,4"
    Rotation="0,10,-10"
    Scale="1,1,1"
   ) {
    rotationMode=BMM_Additive
    translationMode=BMM_Additive
    alphaBoolBlend {blendOption=Linear}
    boneToModify {boneName=dyn_bag_MSTR}
    rotation {roll=-10, yaw=10}
    translation {z=4}
   }
   n2 = Get GravityOverride()
   n3 = RigidBody!(ComponentPose=@n1, Alpha, OverrideWorldGravity:Vector=@n2, ExternalForce) -> (Pose) {
    bClampLinearTranslationLimitToRefPose=true
    bOverrideWorldGravity=true
    cachedBoundsScale=2
    overridePhysicsAsset=&/F_MED_Islander_dyn_skirt_perf
    simulationSpace=BaseBoneSpace
    alphaBoolBlend {blendOption=Linear}
    baseBoneRef {boneName=pelvis}
    componentAppliedLinearAccClamp=<2000 2000 2000>
    componentLinearAccScale=<0.2 0.2 0.4>
    componentLinearVelScale=<0.1 0.1 0.2>
    overrideWorldGravity {z=-300}
    simSpaceSettings {maxAngularAcceleration=10, maxAngularVelocity=10, maxLinearAcceleration=1200, maxLinearVelocity=1200, worldAlpha=1}
   }
   n4 = "Transform (Modify) Bone - Bone: dyn_leaves_l_01"(ComponentPose=@n3, Alpha=1, Translation="2,0,0", Rotation="0,-20,-10", Scale="1,1,1") {
    rotationMode=BMM_Additive
    translationMode=BMM_Additive
    alphaBoolBlend {blendOption=Linear}
    boneToModify {boneName=dyn_leaves_l_01}
    rotation {roll=-10, yaw=-20}
    translation {x=2}
   }
   n5 = "Constraint - dyn_leaves_ft_01 (R)"(ComponentPose=@n4, Alpha=0.8, ConstraintWeights_0=1) {
    alphaBoolBlend {blendOption=Linear}
    boneToModify {boneName=dyn_leaves_ft_01}
    constraintSetup [{transformType=Rotation, perAxis {bX=false, bZ=false}, targetBone {boneName=thigh_r}}]
    constraintWeights [1]
   }
   n6 = "Transform (Modify) Bone - Bone: dyn_leaves_ft_01"(ComponentPose=@n5, Alpha=1, Translation="-2,-2,2", Rotation="0,90,15", Scale="1,1,1") {
    rotationMode=BMM_Additive
    translationMode=BMM_Additive
    alphaBoolBlend {blendOption=Linear}
    boneToModify {boneName=dyn_leaves_ft_01}
    rotation {roll=15, yaw=90}
    translation=<-2 -2 2>
   }
   n7 = "Constraint - dyn_leaves_ft_l_01 (R)"(ComponentPose=@n6, Alpha=0.8, ConstraintWeights_0=1) {
    alphaBoolBlend {blendOption=Linear}
    boneToModify {boneName=dyn_leaves_ft_l_01}
    constraintSetup [{transformType=Rotation, perAxis {bX=false, bZ=false}, targetBone {boneName=thigh_l}}]
    constraintWeights [1]
   }
   n8 = "Transform (Modify) Bone - Bone: dyn_leaves_ft_l_01"(ComponentPose=@n7, Alpha=1, Translation="-1,-3,-1", Rotation="10,42,0", Scale="1,1,1") {
    rotationMode=BMM_Additive
    translationMode=BMM_Additive
    alphaBoolBlend {blendOption=Linear}
    boneToModify {boneName=dyn_leaves_ft_l_01}
    rotation {pitch=10, yaw=42}
    translation=<-1 -3 -1>
   }
   n9 = "Transform (Modify) Bone - Bone: dyn_leaves_ft_l_02"(ComponentPose=@n8, Alpha=1, Translation="0,0,0", Rotation="0,10,0", Scale="1,1,1") {
    rotationMode=BMM_Additive
    translationMode=BMM_Additive
    alphaBoolBlend {blendOption=Linear}
    boneToModify {boneName=dyn_leaves_ft_l_02}
    rotation {yaw=10}
   }
   n10 = "Transform (Modify) Bone - Bone: dyn_leaves_ft_02"(ComponentPose=@n9, Alpha=1, Translation="0,0,0", Rotation="-20,30,-20", Scale="1,1,1") {
    rotationMode=BMM_Additive
    translationMode=BMM_Additive
    alphaBoolBlend {blendOption=Linear}
    boneToModify {boneName=dyn_leaves_ft_02}
    rotation=<rot -20 30 -20>
   }
   n11 = Component To Local(ComponentPose=@n10)
   n12 = RigidBody(ComponentPose=Local To Component(LocalPose=Use cached pose 'SavedPose20243'()), OverrideWorldGravity=@n2) {
    bClampLinearTranslationLimitToRefPose=true
    bOverrideWorldGravity=true
    cachedBoundsScale=2
    overridePhysicsAsset=&F_MED_Islander_dyn_skirt_perf
    simulationSpace=BaseBoneSpace
    alphaBoolBlend {blendOption=Linear}
    baseBoneRef {boneName=pelvis}
    componentAppliedLinearAccClamp=<2000 2000 2000>
    componentLinearAccScale=<0.2 0.2 0.4>
    componentLinearVelScale=<0.1 0.1 0.2>
    overrideWorldGravity {z=-300}
    simSpaceSettings {maxAngularAcceleration=10, maxAngularVelocity=10, maxLinearAcceleration=1200, maxLinearVelocity=1200, worldAlpha=1}
   }
   n13 = "Transform (Modify) Bone - Bone: dyn_leaves_ft_02"(ComponentPose=@n12, Alpha=1, Translation="0,0,0", Rotation="22,8,10", Scale="1,1,1") {
    rotationMode=BMM_Additive
    rotationSpace=BCS_ParentBoneSpace
    translationMode=BMM_Additive
    alphaBoolBlend {blendOption=Linear}
    boneToModify {boneName=dyn_leaves_ft_02}
    rotation=<rot 22 8 10>
   }
   n14 = "Transform (Modify) Bone - Bone: dyn_leaves_ft_l_02"(ComponentPose=@n13, Alpha=1, Translation="0,0,0", Rotation="20,0,0", Scale="1,1,1") {
    rotationMode=BMM_Additive
    rotationSpace=BCS_ParentBoneSpace
    translationMode=BMM_Additive
    alphaBoolBlend {blendOption=Linear}
    boneToModify {boneName=dyn_leaves_ft_l_02}
    rotation {pitch=20}
   }
   n15 = Component To Local(ComponentPose=@n14)
   n16 = Get bIsCrouching()
   n17 = RigidBody!(
    ComponentPose=Local To Component(LocalPose=Blend Poses by bool(BlendPose_0=@n11, BlendPose_1=@n15, bActiveValue=@n16) {blendType=Linear})
    Alpha
    ExternalForce
   ) -> (Pose) {
    bClampLinearTranslationLimitToRefPose=true
    bOverrideWorldGravity=true
    cachedBoundsScale=2
    overridePhysicsAsset=&/F_MED_Islander_dyn_necklace
    simulationSpace=BaseBoneSpace
    alphaBoolBlend {blendOption=Linear}
    baseBoneRef {boneName=spine_05}
    componentAppliedLinearAccClamp=<1500 1500 1500>
    componentLinearAccScale=<0.3 0.3 0.3>
    componentLinearVelScale=<0.3 0.3 0.3>
    overrideWorldGravity {z=-980}
    simSpaceSettings {maxAngularAcceleration=50, maxAngularVelocity=50, maxLinearAcceleration=1000, maxLinearVelocity=1000, velocityScaleZ=2}
   }
   n18 = Anim Dynamics - Bone: dyn_feathers_01(ComponentPose=@n17, Alpha=1, ExternalForce="0,0,0") {
    bAngularSpring=true
    bUseGravityOverride=true
    simulationSpace=RootRelative
    alphaBoolBlend {blendOption=Linear}
    boundBone {boneName=dyn_feathers_01}
    componentAppliedLinearAccClamp=<3000 3000 3000>
    componentLinearAccScale=<0.2 0.2 0.4>
    componentLinearVelScale=<0.1 0.1 0.2>
    gravityOverride {z=-980}
    physicsBodyDefinitions [{
      sphereCollisionRadius=0
      boundBone {boneName=dyn_feathers_01}
      boxExtents=<16 10 6>
      constraintSetup {angularLimitsMax=<0 1 0>, angularLimitsMin=<0 -45 0>, angularTarget=<20 0 0>}
      localJointOffset=<2.5 0 0>
     }]
    relativeSpaceBone {boneName=root}
   }
   n19 = Anim Dynamics - Bone: dyn_feathers_02(ComponentPose=@n18, Alpha=1, ExternalForce="0,0,0") {
    bAngularSpring=true
    bUseGravityOverride=true
    simulationSpace=RootRelative
    alphaBoolBlend {blendOption=Linear}
    boundBone {boneName=dyn_feathers_02}
    componentAppliedLinearAccClamp=<3000 3000 3000>
    componentLinearAccScale=<0.2 0.2 0.4>
    componentLinearVelScale=<0.1 0.1 0.2>
    gravityOverride {z=-980}
    physicsBodyDefinitions [{
      sphereCollisionRadius=0
      boundBone {boneName=dyn_feathers_02}
      boxExtents=<16 10 6>
      constraintSetup {angularLimitsMax=<0 1 0>, angularLimitsMin=<0 -45 0>, angularTarget=<20 0 0>}
      localJointOffset=<2.5 0 0>
     }]
    relativeSpaceBone {boneName=root}
   }
   n20 = Anim Dynamics - Bone: dyn_feathers_03(ComponentPose=@n19, Alpha=1, ExternalForce="0,0,0") {
    bAngularSpring=true
    bUseGravityOverride=true
    simulationSpace=RootRelative
    alphaBoolBlend {blendOption=Linear}
    boundBone {boneName=dyn_feathers_03}
    componentAppliedLinearAccClamp=<3000 3000 3000>
    componentLinearAccScale=<0.2 0.2 0.4>
    componentLinearVelScale=<0.1 0.1 0.2>
    gravityOverride {z=-980}
    physicsBodyDefinitions [{
      sphereCollisionRadius=0
      boundBone {boneName=dyn_feathers_03}
      boxExtents=<16 10 6>
      constraintSetup {angularLimitsMax=<0 1 0>, angularLimitsMin=<0 -45 0>, angularTarget=<20 0 0>}
      localJointOffset=<2.5 0 0>
     }]
    relativeSpaceBone {boneName=root}
   }
   n21 = Anim Dynamics - Bone: dyn_feathers_01(ComponentPose=@n20, Alpha=1, ExternalForce="0,0,0") {
    bAngularSpring=true
    bUseGravityOverride=true
    simulationSpace=RootRelative
    alphaBoolBlend {blendOption=Linear}
    boundBone {boneName=dyn_feathers_01}
    componentAppliedLinearAccClamp=<3000 3000 3000>
    componentLinearAccScale=<0.2 0.2 0.4>
    componentLinearVelScale=<0.1 0.1 0.2>
    gravityOverride {z=-980}
    physicsBodyDefinitions [{
      sphereCollisionRadius=0
      boundBone {boneName=dyn_feathers_01}
      boxExtents=<16 10 6>
      constraintSetup {angularLimitsMax=<0 1 0>, angularLimitsMin=<0 -45 0>, angularTarget=<20 0 0>}
      localJointOffset=<2.5 0 0>
     }]
    relativeSpaceBone {boneName=root}
   }
   n22 = Anim Dynamics - Bone: dyn_feathers_05(ComponentPose=@n21, Alpha=1, ExternalForce="0,0,0") {
    bAngularSpring=true
    bUseGravityOverride=true
    simulationSpace=RootRelative
    alphaBoolBlend {blendOption=Linear}
    boundBone {boneName=dyn_feathers_05}
    componentAppliedLinearAccClamp=<3000 3000 3000>
    componentLinearAccScale=<0.2 0.2 0.4>
    componentLinearVelScale=<0.1 0.1 0.2>
    gravityOverride {z=-980}
    physicsBodyDefinitions [{
      sphereCollisionRadius=0
      boundBone {boneName=dyn_feathers_05}
      boxExtents=<16 10 6>
      constraintSetup {angularLimitsMax=<0 1 0>, angularLimitsMin=<0 -45 0>, angularTarget=<20 0 0>}
      localJointOffset=<2.5 0 0>
     }]
    relativeSpaceBone {boneName=root}
   }
   n23 = Anim Dynamics - Bone: dyn_feathers_06(ComponentPose=@n22, Alpha=1, ExternalForce="0,0,0") {
    bAngularSpring=true
    bUseGravityOverride=true
    simulationSpace=RootRelative
    alphaBoolBlend {blendOption=Linear}
    boundBone {boneName=dyn_feathers_06}
    componentAppliedLinearAccClamp=<3000 3000 3000>
    componentLinearAccScale=<0.2 0.2 0.4>
    componentLinearVelScale=<0.1 0.1 0.2>
    gravityOverride {z=-980}
    physicsBodyDefinitions [{
      sphereCollisionRadius=0
      boundBone {boneName=dyn_feathers_06}
      boxExtents=<16 10 6>
      constraintSetup {angularLimitsMax=<0 1 0>, angularLimitsMin=<0 -45 0>, angularTarget=<20 0 0>}
      localJointOffset=<2.5 0 0>
     }]
    relativeSpaceBone {boneName=root}
   }
   n24 = Anim Dynamics - Bone: dyn_feathers_07(ComponentPose=@n23, Alpha=1, ExternalForce="0,0,0") {
    bAngularSpring=true
    bUseGravityOverride=true
    simulationSpace=RootRelative
    alphaBoolBlend {blendOption=Linear}
    boundBone {boneName=dyn_feathers_07}
    componentAppliedLinearAccClamp=<3000 3000 3000>
    componentLinearAccScale=<0.2 0.2 0.4>
    componentLinearVelScale=<0.1 0.1 0.2>
    gravityOverride {z=-980}
    physicsBodyDefinitions [{
      sphereCollisionRadius=0
      boundBone {boneName=dyn_feathers_07}
      boxExtents=<16 10 6>
      constraintSetup {angularLimitsMax=<0 1 0>, angularLimitsMin=<0 -45 0>, angularTarget=<20 0 0>}
      localJointOffset=<2.5 0 0>
     }]
    relativeSpaceBone {boneName=root}
   }
   n25 = Anim Dynamics - Bone: dyn_feathers_08(ComponentPose=@n24, Alpha=1, ExternalForce="0,0,0") {
    bAngularSpring=true
    bUseGravityOverride=true
    simulationSpace=RootRelative
    alphaBoolBlend {blendOption=Linear}
    boundBone {boneName=dyn_feathers_08}
    componentAppliedLinearAccClamp=<3000 3000 3000>
    componentLinearAccScale=<0.2 0.2 0.4>
    componentLinearVelScale=<0.1 0.1 0.2>
    gravityOverride {z=-980}
    physicsBodyDefinitions [{
      sphereCollisionRadius=0
      boundBone {boneName=dyn_feathers_08}
      boxExtents=<16 10 6>
      constraintSetup {angularLimitsMax=<0 1 0>, angularLimitsMin=<0 -45 0>, angularTarget=<20 0 0>}
      localJointOffset=<2.5 0 0>
     }]
    relativeSpaceBone {boneName=root}
   }
   n26 = Anim Dynamics - Bone: dyn_feathers_09(ComponentPose=@n25, Alpha=1, ExternalForce="0,0,0") {
    bAngularSpring=true
    bUseGravityOverride=true
    simulationSpace=RootRelative
    alphaBoolBlend {blendOption=Linear}
    boundBone {boneName=dyn_feathers_09}
    componentAppliedLinearAccClamp=<3000 3000 3000>
    componentLinearAccScale=<0.2 0.2 0.4>
    componentLinearVelScale=<0.1 0.1 0.2>
    gravityOverride {z=-980}
    physicsBodyDefinitions [{
      sphereCollisionRadius=0
      boundBone {boneName=dyn_feathers_09}
      boxExtents=<16 10 6>
      constraintSetup {angularLimitsMax=<0 1 0>, angularLimitsMin=<0 -45 0>, angularTarget=<20 0 0>}
      localJointOffset=<2.5 0 0>
     }]
    relativeSpaceBone {boneName=root}
   }
   Output Pose(Result=Component To Local(ComponentPose=@n26)) {layerGroup=None}
   n27 = "Constraint - dyn_arm_feather_l_MSTR (PPPP)"!(
    ComponentPose=Local To Component(LocalPose=Copy Pose From Mesh() {bUseAttachedParent=true})
    Alpha=1
    ConstraintWeights_0=0.0439
    ConstraintWeights_1:float=0.9184
    ConstraintWeights_2:float=0
    ConstraintWeights_3:float=0.0377
   ) -> (Pose) {
    alphaBoolBlend {blendOption=Linear}
    boneToModify {boneName=dyn_arm_feather_l_MSTR}
    constraintSetup [
     {targetBone {boneName=upperarm_twist_02_l}}
     {targetBone {boneName=upperarm_twist_01_l}}
     {targetBone {boneName=clavicle_l}}
     {targetBone {boneName=upperarm_l}}
    ]
    constraintWeights [1,1,1,1]
   }
   n28 = "Constraint - dyn_arm_feather_r_MSTR (PPPP)"!(
    ComponentPose=@n27
    Alpha=1
    ConstraintWeights_0=0
    ConstraintWeights_1:float=0.9183
    ConstraintWeights_2:float=0.0377
    ConstraintWeights_3:float=0.044
   ) -> (Pose) {
    alphaBoolBlend {blendOption=Linear}
    boneToModify {boneName=dyn_arm_feather_r_MSTR}
    constraintSetup [
     {targetBone {boneName=clavicle_r}}
     {targetBone {boneName=upperarm_twist_01_r}}
     {targetBone {boneName=upperarm_r}}
     {targetBone {boneName=upperarm_twist_02_r}}
    ]
    constraintWeights [1,1,1,1]
   }
   n29 = "Constraint - dyn_feathers_01_MSTR (PPPPP)"!(
    ComponentPose=@n28
    Alpha=1
    ConstraintWeights_0=0.242
    ConstraintWeights_1:float=0.3879
    ConstraintWeights_2:float=0.2122
    ConstraintWeights_3:float=0.0442
    ConstraintWeights_4:float=0.1138
   ) -> (Pose) {
    alphaBoolBlend {blendOption=Linear}
    boneToModify {boneName=dyn_feathers_01_MSTR}
    constraintSetup [
     {targetBone {boneName=spine_01}}
     {targetBone {boneName=spine_02}}
     {targetBone {boneName=spine_03}}
     {targetBone {boneName=spine_04}}
     {targetBone {boneName=pelvis}}
    ]
    constraintWeights [1,1,1,1,1]
   }
   n30 = "Constraint - dyn_feathers_02_MSTR (PPPPP)"!(
    ComponentPose=@n29
    Alpha=1
    ConstraintWeights_0=0.2136
    ConstraintWeights_1:float=0.3234
    ConstraintWeights_2:float=0.1807
    ConstraintWeights_3:float=0.0352
    ConstraintWeights_4:float=0.2471
   ) -> (Pose) {
    alphaBoolBlend {blendOption=Linear}
    boneToModify {boneName=dyn_feathers_02_MSTR}
    constraintSetup [
     {targetBone {boneName=spine_01}}
     {targetBone {boneName=spine_02}}
     {targetBone {boneName=spine_03}}
     {targetBone {boneName=spine_04}}
     {targetBone {boneName=pelvis}}
    ]
    constraintWeights [1,1,1,1,1]
   }
   n31 = "Constraint - dyn_feathers_03_MSTR (PPPPP)"!(
    ComponentPose=@n30
    Alpha=1
    ConstraintWeights_0=0.2006
    ConstraintWeights_1:float=0.3093
    ConstraintWeights_2:float=0.1419
    ConstraintWeights_3:float=0.0257
    ConstraintWeights_4:float=0.3225
   ) -> (Pose) {
    alphaBoolBlend {blendOption=Linear}
    boneToModify {boneName=dyn_feathers_03_MSTR}
    constraintSetup [
     {targetBone {boneName=spine_01}}
     {targetBone {boneName=spine_02}}
     {targetBone {boneName=spine_03}}
     {targetBone {boneName=spine_04}}
     {targetBone {boneName=pelvis}}
    ]
    constraintWeights [1,1,1,1,1]
   }
   n32 = "Constraint - dyn_feathers_04_MSTR (PPPPP)"!(
    ComponentPose=@n31
    Alpha=1
    ConstraintWeights_0=0.192
    ConstraintWeights_1:float=0.3008
    ConstraintWeights_2:float=0.1201
    ConstraintWeights_3:float=0.025
    ConstraintWeights_4:float=0.362
   ) -> (Pose) {
    alphaBoolBlend {blendOption=Linear}
    boneToModify {boneName=dyn_feathers_04_MSTR}
    constraintSetup [
     {targetBone {boneName=spine_01}}
     {targetBone {boneName=spine_02}}
     {targetBone {boneName=spine_03}}
     {targetBone {boneName=spine_04}}
     {targetBone {boneName=pelvis}}
    ]
    constraintWeights [1,1,1,1,1]
   }
   n33 = "Constraint - dyn_feathers_05_MSTR (PPPPP)"!(
    ComponentPose=@n32
    Alpha=1
    ConstraintWeights_0=0.2021
    ConstraintWeights_1:float=0.2933
    ConstraintWeights_2:float=0.1234
    ConstraintWeights_3:float=0.0319
    ConstraintWeights_4:float=0.3494
   ) -> (Pose) {
    alphaBoolBlend {blendOption=Linear}
    boneToModify {boneName=dyn_feathers_05_MSTR}
    constraintSetup [
     {targetBone {boneName=spine_01}}
     {targetBone {boneName=spine_02}}
     {targetBone {boneName=spine_03}}
     {targetBone {boneName=spine_04}}
     {targetBone {boneName=pelvis}}
    ]
    constraintWeights [1,1,1,1,1]
   }
   n34 = "Constraint - dyn_feathers_06_MSTR (PPPPP)"!(
    ComponentPose=@n33
    Alpha=1
    ConstraintWeights_0=0.2067
    ConstraintWeights_1:float=0.3012
    ConstraintWeights_2:float=0.1192
    ConstraintWeights_3:float=0.0269
    ConstraintWeights_4:float=0.3461
   ) -> (Pose) {
    alphaBoolBlend {blendOption=Linear}
    boneToModify {boneName=dyn_feathers_06_MSTR}
    constraintSetup [
     {targetBone {boneName=spine_01}}
     {targetBone {boneName=spine_02}}
     {targetBone {boneName=spine_03}}
     {targetBone {boneName=spine_04}}
     {targetBone {boneName=pelvis}}
    ]
    constraintWeights [1,1,1,1,1]
   }
   n35 = "Constraint - dyn_feathers_07_MSTR (PPPPP)"!(
    ComponentPose=@n34
    Alpha=1
    ConstraintWeights_0=0.2051
    ConstraintWeights_1:float=0.3147
    ConstraintWeights_2:float=0.1449
    ConstraintWeights_3:float=0.0266
    ConstraintWeights_4:float=0.3088
   ) -> (Pose) {
    alphaBoolBlend {blendOption=Linear}
    boneToModify {boneName=dyn_feathers_07_MSTR}
    constraintSetup [
     {targetBone {boneName=spine_01}}
     {targetBone {boneName=spine_02}}
     {targetBone {boneName=spine_03}}
     {targetBone {boneName=spine_04}}
     {targetBone {boneName=pelvis}}
    ]
    constraintWeights [1,1,1,1,1]
   }
   n36 = "Constraint - dyn_feathers_08_MSTR (PPPPPP)"!(
    ComponentPose=@n35
    Alpha=1
    ConstraintWeights_0=0.0905
    ConstraintWeights_1:float=0.2183
    ConstraintWeights_2:float=0.3096
    ConstraintWeights_3:float=0.1552
    ConstraintWeights_4:float=0.0264
    ConstraintWeights_5:float=0.2001
   ) -> (Pose) {
    alphaBoolBlend {blendOption=Linear}
    boneToModify {boneName=dyn_feathers_08_MSTR}
    constraintSetup [
     {targetBone {boneName=dyn_leaves_l_01}}
     {targetBone {boneName=spine_01}}
     {targetBone {boneName=spine_02}}
     {targetBone {boneName=spine_03}}
     {targetBone {boneName=spine_04}}
     {targetBone {boneName=pelvis}}
    ]
    constraintWeights [1,1,1,1,1,1]
   }
   n37 = "Constraint - dyn_feathers_09_MSTR (PPPPP)"!(
    ComponentPose=@n36
    Alpha=1
    ConstraintWeights_0=0.2373
    ConstraintWeights_1:float=0.3613
    ConstraintWeights_2:float=0.1746
    ConstraintWeights_3:float=0.0335
    ConstraintWeights_4:float=0.1933
   ) -> (Pose) {
    alphaBoolBlend {blendOption=Linear}
    boneToModify {boneName=dyn_feathers_09_MSTR}
    constraintSetup [
     {targetBone {boneName=spine_01}}
     {targetBone {boneName=spine_02}}
     {targetBone {boneName=spine_03}}
     {targetBone {boneName=spine_04}}
     {targetBone {boneName=pelvis}}
    ]
    constraintWeights [1,1,1,1,1]
   }
   n38 = "Constraint - dyn_leaves_bk_r_MSTR (PP)"!(ComponentPose=@n37, Alpha=1, ConstraintWeights_0=0.7, ConstraintWeights_1:float=0.3) -> (Pose) {
    alphaBoolBlend {blendOption=Linear}
    boneToModify {boneName=dyn_leaves_bk_r_MSTR}
    constraintSetup [{targetBone {boneName=thigh_twist_01_r}}, {targetBone {boneName=pelvis}}]
    constraintWeights [1,1]
   }
   n39 = "Constraint - dyn_pompon_MSTR (PPPP)"!(
    ComponentPose=@n38
    Alpha=1
    ConstraintWeights_0=0.28
    ConstraintWeights_1:float=0.36
    ConstraintWeights_2:float=0.23
    ConstraintWeights_3:float=0.13
   ) -> (Pose) {
    alphaBoolBlend {blendOption=Linear}
    boneToModify {boneName=dyn_pompon_MSTR}
    constraintSetup [{targetBone {boneName=spine_01}}, {targetBone {boneName=spine_02}}, {targetBone {boneName=spine_03}}, {targetBone {boneName=pelvis}}]
    constraintWeights [1,1,1,1]
   }
   n40 = "Constraint - dyn_bag_MSTR (PP)"!(ComponentPose=@n39, Alpha=1, ConstraintWeights_0=0.8, ConstraintWeights_1:float=0.2) -> (Pose) {
    alphaBoolBlend {blendOption=Linear}
    boneToModify {boneName=dyn_bag_MSTR}
    constraintSetup [{targetBone {boneName=pelvis}}, {targetBone {boneName=thigh_r}}]
    constraintWeights [1,1]
   }
   n41 = Get ThighRotationUpLeft() -> (ThighRotationUpLeft:float)
   n42 = "Transform (Modify) Bone - Bone: dyn_leaves_ft_01"(ComponentPose=@n40, Alpha=@n41, Translation="2,0,0", Rotation="10.000,0.000,15", Scale="1,1,1") {
    rotationMode=BMM_Additive
    rotationSpace=BCS_BoneSpace
    translationMode=BMM_Additive
    translationSpace=BCS_BoneSpace
    alphaBoolBlend {blendOption=Linear}
    boneToModify {boneName=dyn_leaves_ft_01}
    rotation {pitch=10, roll=15}
    translation {x=2}
   }
   n43 = "Transform (Modify) Bone - Bone: dyn_leaves_l_01"(ComponentPose=@n42, Alpha=@n41, Translation="-0.860,-0.000,1.144", Rotation="10.000,0.000,-20", Scale="1,1,1") {
    rotationMode=BMM_Additive
    rotationSpace=BCS_BoneSpace
    translationMode=BMM_Additive
    translationSpace=BCS_BoneSpace
    alphaBoolBlend {blendOption=Linear}
    boneToModify {boneName=dyn_leaves_l_01}
    rotation {pitch=10, roll=-20}
    translation {x=-0.86, z=1.144}
   }
   n44 = "Transform (Modify) Bone - Bone: dyn_feathers_05_MSTR"(ComponentPose=@n43, Alpha=@n41, Translation="0,0,0", Rotation="60,0,0", Scale="1,1,1") {
    rotationMode=BMM_Additive
    rotationSpace=BCS_BoneSpace
    translationMode=BMM_Additive
    translationSpace=BCS_BoneSpace
    alphaBoolBlend {blendOption=Linear}
    boneToModify {boneName=dyn_feathers_05_MSTR}
    rotation {pitch=60}
   }
   n45 = "Transform (Modify) Bone - Bone: dyn_feathers_06_MSTR"(ComponentPose=@n44, Alpha=@n41, Translation="0,0,0", Rotation="60,0,0", Scale="1,1,1") {
    rotationMode=BMM_Additive
    rotationSpace=BCS_BoneSpace
    translationMode=BMM_Additive
    translationSpace=BCS_BoneSpace
    alphaBoolBlend {blendOption=Linear}
    boneToModify {boneName=dyn_feathers_06_MSTR}
    rotation {pitch=60}
   }
   n46 = "Transform (Modify) Bone - Bone: dyn_feathers_07_MSTR"(ComponentPose=@n45, Alpha=@n41, Translation="0,0,0", Rotation="45,0,0", Scale="1,1,1") {
    rotationMode=BMM_Additive
    rotationSpace=BCS_BoneSpace
    translationMode=BMM_Additive
    translationSpace=BCS_BoneSpace
    alphaBoolBlend {blendOption=Linear}
    boneToModify {boneName=dyn_feathers_07_MSTR}
    rotation {pitch=45}
   }
   n47 = "Transform (Modify) Bone - Bone: dyn_feathers_08_MSTR"(ComponentPose=@n46, Alpha=@n41, Translation="0,0,0", Rotation="60,0,0", Scale="1,1,1") {
    rotationMode=BMM_Additive
    rotationSpace=BCS_BoneSpace
    translationMode=BMM_Additive
    translationSpace=BCS_BoneSpace
    alphaBoolBlend {blendOption=Linear}
    boneToModify {boneName=dyn_feathers_08_MSTR}
    rotation {pitch=60}
   }
   n48 = "Transform (Modify) Bone - Bone: dyn_bag_01"(
    ComponentPose=@n47
    Alpha=Get ThighRotationUpRight() -> (ThighRotationUpRight:float)
    Translation="-5.5,0,0"
    Rotation="0,0,0"
    Scale="1,1,1"
   ) {
    translationMode=BMM_Additive
    translationSpace=BCS_BoneSpace
    alphaBoolBlend {blendOption=Linear}
    boneToModify {boneName=dyn_bag_01}
    translation {x=-5.5}
   }
   n49 = "Transform (Modify) Bone - Bone: dyn_skirt_bk_l_MSTR"(ComponentPose=@n48, Alpha=Get KneeRotationUpLeft() -> (KneeRotationUpLeft:float), Translation="-15,0,0", Rotation="0,-10,0", Scale="1,1,1") {
    rotationMode=BMM_Additive
    rotationSpace=BCS_BoneSpace
    translationMode=BMM_Additive
    translationSpace=BCS_BoneSpace
    alphaBoolBlend {blendOption=Linear}
    boneToModify {boneName=dyn_skirt_bk_l_MSTR}
    rotation {yaw=-10}
    translation {x=-15}
   }
   n50 = "Transform (Modify) Bone - Bone: dyn_skirt_bk_r_MSTR"(ComponentPose=@n49, Alpha=Get KneeRotationUpRight() -> (KneeRotationUpRight:float), Translation="-15,0,0", Rotation="0,0,0", Scale="1,1,1") {
    rotationMode=BMM_Additive
    rotationSpace=BCS_BoneSpace
    scaleMode=BMM_Additive
    scaleSpace=BCS_BoneSpace
    translationMode=BMM_Additive
    translationSpace=BCS_BoneSpace
    alphaBoolBlend {blendOption=Linear}
    boneToModify {boneName=dyn_skirt_bk_r_MSTR}
    translation {x=-15}
   }
   Save cached pose 'SavedPose20243'(Pose=Component To Local(ComponentPose=@n50)) {cacheName=SavedPose20243}
  ]}
 EventGraph {nodes [
   Event OnParachuteOpened()
   > n1 = Delay(Duration=1.5)
   ResetDynamics(execute=[@n1.then, @n2.then, @n3.then], InTeleportType=TeleportPhysics)
   > ResetDynamics(InTeleportType=ResetPhysics)
   n2 = Event OnBeginSkydiving()
   n3 = Event OnEndSkydiving()
   Delay(Duration=0.05)
   Event OnBeginCrouching()
   Event OnEndCrouching()
  ]}
]
variables [KneeRotationUpRight_old {type=number}]
