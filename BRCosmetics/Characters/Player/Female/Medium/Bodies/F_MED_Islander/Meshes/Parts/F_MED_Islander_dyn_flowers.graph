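# Physics asset for the flower-and-leaf part of the female medium Islander player cosmetic, defining small capsule bodies and joints for its dyn_ bones.
# A master body anchors three single-leaf branches plus two short two-link chains: leaves 02 to 03 and flowers 01 to 02.
# Every constraint uses narrow swing limits and almost no twist, restricting the pieces to slight bending.
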PhysicsAsset /BRCosmetics/Characters/Player/Female/Medium/Bodies/F_MED_Islander/Meshes/Parts/F_MED_Islander_dyn_flowers
bodies {
 dyn_flowers_01 [{shapeType=Capsule, length=1, radius=1, center=<1 0 0>, rotation=<rot 90 0 0>}]
 dyn_flowers_02 [{shapeType=Capsule, length=1, radius=1, center=<1 0 0>, rotation=<rot 90 0 0>}]
 dyn_flowers_MSTR [{shapeType=Capsule, length=0.5, radius=0.5}]
 dyn_leaves_01 [{shapeType=Capsule, length=4, radius=1, shapeName=dyn_leaves_01_capsule, center=<3 0 0>, rotation=<rot 90 0 0>}]
 dyn_leaves_02 [{shapeType=Capsule, length=4, radius=1, shapeName=dyn_leaves_02_capsule, center=<1 0 0>, rotation=<rot 90 0 0>}]
 dyn_leaves_03 [{shapeType=Capsule, length=4, radius=1, shapeName=dyn_leaves_03_capsule, center=<1.9749 0 0>, rotation=<rot 90 0 0>}]
 dyn_leaves_04 [{shapeType=Capsule, length=4, radius=1, shapeName=dyn_leaves_04_capsule, center=<3 0 0>, rotation=<rot 90 0 0>}]
 dyn_leaves_05 [{shapeType=Capsule, length=4, radius=1, shapeName=dyn_leaves_05_capsule, center=<3 0 0>, rotation=<rot 90 0 0>}]
}
constraints [
 {bone1Name=dyn_leaves_01, bone2Name=dyn_flowers_MSTR, swing1LimitDegrees=20, swing2LimitDegrees=8, twistLimitDegrees=1}
 {bone1Name=dyn_leaves_02, bone2Name=dyn_flowers_MSTR, swing1LimitDegrees=12, swing2LimitDegrees=8, twistLimitDegrees=1}
 {bone1Name=dyn_leaves_03, bone2Name=dyn_leaves_02, swing1LimitDegrees=12, swing2LimitDegrees=8, twistLimitDegrees=1}
 {bone1Name=dyn_leaves_04, bone2Name=dyn_flowers_MSTR, swing1LimitDegrees=20, swing2LimitDegrees=8, twistLimitDegrees=1}
 {bone1Name=dyn_leaves_05, bone2Name=dyn_flowers_MSTR, swing1LimitDegrees=20, swing2LimitDegrees=8, twistLimitDegrees=1}
 {bone1Name=dyn_flowers_01, bone2Name=dyn_flowers_MSTR, swing1LimitDegrees=15, swing2LimitDegrees=8, twistLimitDegrees=1}
 {bone1Name=dyn_flowers_02, bone2Name=dyn_flowers_01, swing1LimitDegrees=15, swing2LimitDegrees=8, twistLimitDegrees=1}
]
