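# Physics asset for the female medium Islander cosmetic body, simulating dangling earrings and left and right hair chains.
# Capsules on head, neck, upper spine and clavicles form the body volume, joined by free constraints.
# Limited-swing capsule chains hang from head-locked master bones, while earrings attach directly to the head.
# Back hair uses mostly flat boxes with wider swing, a left-side branch, and masters without a constraint to the head.
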PhysicsAsset /BRCosmetics/Characters/Player/Female/Medium/Bodies/F_MED_Islander/Meshes/Parts/F_MED_Islander_dyn_front_perf
bodies {
 clavicle_l [{shapeType=Capsule, length=8.75, radius=5.5, shapeName=clavicle_l_capsule, center=<5 -1.4671 0>, rotation=<rot 84.9998 0 0>}]
 clavicle_r [{shapeType=Capsule, length=8.75, radius=5.5, shapeName=clavicle_r_capsule, center=<-5 1.6081 0>, rotation=<rot 84.9998 0 0>}]
 dyn_earrings_l [{shapeType=Capsule, length=2.5, radius=0.6, shapeName=dyn_earrings_l_capsule, center=<1.2706 0 0.0002>, rotation=<rot 0 0 89.9999>}]
 dyn_earrings_r [{shapeType=Capsule, length=2.5, radius=0.6, shapeName=dyn_earrings_r_capsule, center=<1.2706 0 0.0002>, rotation=<rot 0 0 89.9999>}]
 dyn_hair_l_01 [{shapeType=Capsule, length=4, radius=1.5, shapeName=dyn_hair_l_01_capsule, center=<3.8423 0.5153 0>, rotation=<rot 90 0 0>}]
 dyn_hair_l_02 [{shapeType=Capsule, length=4, radius=1.5, shapeName=dyn_hair_l_02_capsule, center=<1.5 0.5153 0>, rotation=<rot 90 0 0>}]
 dyn_hair_l_03 [{shapeType=Capsule, length=3, radius=1.5, shapeName=dyn_hair_l_03_capsule, center=<1.5 0.5153 0>, rotation=<rot 90 0 0>}]
 dyn_hair_l_04 [{shapeType=Capsule, length=2.125, radius=1.375, shapeName=dyn_hair_l_04_capsule, center=<2 1 0>, rotation=<rot 90 0 0>}]
 dyn_hair_l_MSTR [{shapeType=Sphere, radius=0.5, shapeName=dyn_hair_l_MSTR_sphere}]
 dyn_hair_l_bk_01 [{shapeType=Box, extentX=8.5, extentY=11.5, extentZ=2, center=<4 0 1.5>, rotation=<rot 4.9238 -0.8704 -10.0385>}]
 dyn_hair_l_bk_02 [{shapeType=Box, extentX=8, extentY=11.5, extentZ=2, center=<3 0 1.5>, rotation=<rot -4.9237 0.8708 -10.0419>}]
 dyn_hair_l_bk_03 [{shapeType=Box, extentX=7.5, extentY=12.5, extentZ=2, center=<3 0 0.5>, rotation=<rot -9.8465 1.7538 -10.1529>}]
 dyn_hair_l_bk_04 [{shapeType=Box, extentX=6, extentY=11, extentZ=2, center=<3 0 0>}]
 dyn_hair_l_bk_05 [{shapeType=Box, extentX=6, extentY=6.5, extentZ=1.5, center=<3 0 0>}]
 dyn_hair_l_bk_06 [
  {shapeType=Capsule, length=3.25, radius=1, center=<2.7933 0.3328 0>, rotation=<rot 90 0 -20.0035>}
  {shapeType=Capsule, length=0.75, radius=0.75, center=<0.9662 1.1999 -1.1387>, rotation=<rot 74.9998 20.002 0>}
 ]
 dyn_hair_l_bk_MSTR [{shapeType=Capsule, length=0.5, radius=0.5}]
 dyn_hair_r_01 [{shapeType=Capsule, length=4, radius=1.5, shapeName=dyn_hair_r_01_capsule, center=<3.8423 -0.2465 0>, rotation=<rot 90 0 0>}]
 dyn_hair_r_02 [{shapeType=Capsule, length=4, radius=1.5, shapeName=dyn_hair_r_02_capsule, center=<1.5 -0.2465 0>, rotation=<rot 90 0 0>}]
 dyn_hair_r_03 [{shapeType=Capsule, length=3, radius=1.5, shapeName=dyn_hair_r_03_capsule, center=<1.5 -0.2465 0>, rotation=<rot 90 0 0>}]
 dyn_hair_r_04 [{shapeType=Capsule, length=1.8125, radius=1.6875, shapeName=dyn_hair_r_04_capsule, center=<2 -1 0>, rotation=<rot 90 0 0>}]
 dyn_hair_r_MSTR [{shapeType=Sphere, radius=0.5, shapeName=dyn_hair_r_MSTR_sphere}]
 dyn_hair_r_bk_01 [{shapeType=Box, extentX=8.5, extentY=11.5, extentZ=2, center=<4 0 1.5>, rotation=<rot 9.9808 -0.4419 -5.0579>}]
 dyn_hair_r_bk_02 [{shapeType=Box, extentX=8, extentY=11.5, extentZ=2, center=<3 0 1.5>}]
 dyn_hair_r_bk_03 [
  {shapeType=Capsule, length=3.25, radius=1, center=<1.9347 -3.229 -1.2675>, rotation=<rot 85.0001 174.9998 -179.9999>}
  {shapeType=Box, extentX=6.5, extentY=12.5, extentZ=2.5, center=<3 0 0.5>, rotation=<rot -10 0 0>}
 ]
 dyn_hair_r_bk_04 [{shapeType=Box, extentX=6, extentY=11, extentZ=2, center=<3 0 0>}]
 dyn_hair_r_bk_05 [{shapeType=Box, extentX=6, extentY=6.5, extentZ=1.5, center=<3 0 0>}]
 dyn_hair_r_bk_MSTR [{shapeType=Capsule, length=0.5, radius=0.5}]
 head [
  {shapeType=Capsule, length=5.25, radius=5.5, center=<3.195 0.6049 0>, rotation=<rot 90 0 0>}
  {shapeType=Capsule, length=9.0021, radius=6.25, center=<0.6916 3.6994 0>, rotation=<rot 90 0 0>}
 ]
 neck_01 [{shapeType=Capsule, length=3.5, radius=4.5, center=<2.6439 -1.473 0>, rotation=<rot 90 -172.875 -166.7599>}]
 spine_04 [{shapeType=Capsule, length=15.0812, radius=8.25, center=<5.5511 1.1165 -0.0001>}]
 spine_05 [{shapeType=Capsule, length=10.1234, radius=6.0452, center=<6.3352 -1 0>}]
}
constraints [
 {
  bone1Name=head
  bone2Name=neck_01
  swing1LimitDegrees=45
  swing1Motion=Free
  swing2LimitDegrees=45
  swing2Motion=Free
  twistLimitDegrees=45
  twistMotion=Free
 }
 {
  bone1Name=neck_01
  bone2Name=spine_05
  swing1LimitDegrees=45
  swing1Motion=Free
  swing2LimitDegrees=45
  swing2Motion=Free
  twistLimitDegrees=45
  twistMotion=Free
 }
 {
  bone1Name=spine_05
  bone2Name=spine_04
  swing1LimitDegrees=45
  swing1Motion=Free
  swing2LimitDegrees=45
  swing2Motion=Free
  twistLimitDegrees=45
  twistMotion=Free
 }
 {bone1Name=dyn_earrings_l, bone2Name=head, swing1LimitDegrees=20, swing2LimitDegrees=10, twistLimitDegrees=2}
 {bone1Name=dyn_earrings_r, bone2Name=head, swing1LimitDegrees=20, swing2LimitDegrees=10, twistLimitDegrees=2}
 {
  bone1Name=dyn_hair_l_MSTR
  bone2Name=head
  swing1LimitDegrees=2
  swing1Motion=Locked
  swing2LimitDegrees=20
  swing2Motion=Locked
  twistLimitDegrees=2
  twistMotion=Locked
 }
 {bone1Name=dyn_hair_l_01, bone2Name=dyn_hair_l_MSTR, swing1LimitDegrees=10, swing2LimitDegrees=5, twistLimitDegrees=1}
 {bone1Name=dyn_hair_l_02, bone2Name=dyn_hair_l_01, swing1LimitDegrees=10, swing2LimitDegrees=5, twistLimitDegrees=1}
 {bone1Name=dyn_hair_l_03, bone2Name=dyn_hair_l_02, swing1LimitDegrees=10, swing2LimitDegrees=5, twistLimitDegrees=1}
 {bone1Name=dyn_hair_l_04, bone2Name=dyn_hair_l_03, swing1LimitDegrees=20, swing2LimitDegrees=15, twistLimitDegrees=5}
 {bone1Name=dyn_hair_r_bk_01, bone2Name=dyn_hair_r_bk_MSTR, swing1LimitDegrees=20, swing2LimitDegrees=45, twistLimitDegrees=1}
 {bone1Name=dyn_hair_r_bk_02, bone2Name=dyn_hair_r_bk_01, swing1LimitDegrees=20, swing2LimitDegrees=45, twistLimitDegrees=1}
 {bone1Name=dyn_hair_r_bk_03, bone2Name=dyn_hair_r_bk_02, swing1LimitDegrees=20, swing2LimitDegrees=45, twistLimitDegrees=1}
 {bone1Name=dyn_hair_r_bk_04, bone2Name=dyn_hair_r_bk_03, swing1LimitDegrees=20, swing2LimitDegrees=45, twistLimitDegrees=1}
 {bone1Name=dyn_hair_r_bk_05, bone2Name=dyn_hair_r_bk_04, swing1LimitDegrees=20, swing2LimitDegrees=45, twistLimitDegrees=1}
 {
  bone1Name=dyn_hair_r_MSTR
  bone2Name=head
  swing1LimitDegrees=2
  swing1Motion=Locked
  swing2LimitDegrees=20
  swing2Motion=Locked
  twistLimitDegrees=2
  twistMotion=Locked
 }
 {bone1Name=dyn_hair_r_01, bone2Name=dyn_hair_r_MSTR, swing1LimitDegrees=10, swing2LimitDegrees=5, twistLimitDegrees=1}
 {bone1Name=dyn_hair_r_02, bone2Name=dyn_hair_r_01, swing1LimitDegrees=10, swing2LimitDegrees=5, twistLimitDegrees=1}
 {bone1Name=dyn_hair_r_03, bone2Name=dyn_hair_r_02, swing1LimitDegrees=10, swing2LimitDegrees=5, twistLimitDegrees=1}
 {bone1Name=dyn_hair_r_04, bone2Name=dyn_hair_r_03, swing1LimitDegrees=20, swing2LimitDegrees=15, twistLimitDegrees=5}
 {bone1Name=dyn_hair_l_bk_01, bone2Name=dyn_hair_l_bk_MSTR, swing1LimitDegrees=20, swing2LimitDegrees=45, twistLimitDegrees=1}
 {bone1Name=dyn_hair_l_bk_02, bone2Name=dyn_hair_l_bk_01, swing1LimitDegrees=20, swing2LimitDegrees=45, twistLimitDegrees=1}
 {bone1Name=dyn_hair_l_bk_03, bone2Name=dyn_hair_l_bk_02, swing1LimitDegrees=20, swing2LimitDegrees=45, twistLimitDegrees=1}
 {bone1Name=dyn_hair_l_bk_04, bone2Name=dyn_hair_l_bk_03, swing1LimitDegrees=20, swing2LimitDegrees=45, twistLimitDegrees=1}
 {bone1Name=dyn_hair_l_bk_05, bone2Name=dyn_hair_l_bk_04, swing1LimitDegrees=20, swing2LimitDegrees=45, twistLimitDegrees=1}
 {bone1Name=dyn_hair_l_bk_06, bone2Name=dyn_hair_l_bk_03, swing1LimitDegrees=20, swing2LimitDegrees=45, twistLimitDegrees=1}
 {
  bone1Name=clavicle_l
  bone2Name=spine_05
  swing1LimitDegrees=45
  swing1Motion=Free
  swing2LimitDegrees=45
  swing2Motion=Free
  twistLimitDegrees=45
  twistMotion=Free
 }
 {
  bone1Name=clavicle_r
  bone2Name=spine_05
  swing1LimitDegrees=45
  swing1Motion=Free
  swing2LimitDegrees=45
  swing2Motion=Free
  twistLimitDegrees=45
  twistMotion=Free
 }
]
