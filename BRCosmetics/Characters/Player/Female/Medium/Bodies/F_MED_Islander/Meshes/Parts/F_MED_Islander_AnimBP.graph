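# Post-body animation Blueprint for the F_MED_Islander part mesh, animating its hair, leaves and flowers.
# It copies the parent pose, then picks a RigidBody hair branch for skydiving, front end or gameplay, using bone constraints and additive poses.
# AnimDynamics then springs leaves and flowers.
# The event graph sets external force from swim, skydive and parachute states, and resets dynamics after parachute opening; head-driven hair lift nodes are unconnected.
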AnimBlueprint /BRCosmetics/Characters/Player/Female/Medium/Bodies/F_MED_Islander/Meshes/Parts/F_MED_Islander_AnimBP
target_skeleton=&/F_MED_Islander_Skeleton
graphs [
 AnimGraph {nodes [
   Component To Local()
   n1 = Get head_rot_up_r() -> (head_rot_up_r:double)
   n2 = "Transform (Modify) Bone - Bone: dyn_hair_l_bk_MSTR"(Alpha=@n1, Translation="0,0,2", Rotation="60,0.000,0.000", Scale="1,1,1") {
    rotationMode=BMM_Additive
    rotationSpace=BCS_BoneSpace
    translationMode=BMM_Additive
    translationSpace=BCS_BoneSpace
    alphaBoolBlend {blendOption=Linear}
    boneToModify {boneName=dyn_hair_l_bk_MSTR}
    rotation {pitch=60}
    translation {z=2}
   }
   "Transform (Modify) Bone - Bone: dyn_hair_r_bk_MSTR"(ComponentPose=@n2, Alpha=@n1, Translation="0,0,2", Rotation="60,0,0", Scale="1,1,1") {
    rotationMode=BMM_Additive
    rotationSpace=BCS_BoneSpace
    translationMode=BMM_Additive
    translationSpace=BCS_BoneSpace
    alphaBoolBlend {blendOption=Linear}
    boneToModify {boneName=dyn_hair_r_bk_MSTR}
    rotation {pitch=60}
    translation {z=2}
   }
   RigidBody() {
    bClampLinearTranslationLimitToRefPose=true
    bOverrideWorldGravity=true
    cachedBoundsScale=2
    overridePhysicsAsset=&/F_MED_Islander_dyn_flowers
    simulationSpace=BaseBoneSpace
    alphaBoolBlend {blendOption=Linear}
    baseBoneRef {boneName=head}
    componentLinearAccScale=<0.2 0.2 0.4>
    componentLinearVelScale=<0.1 0.1 0.2>
    overrideWorldGravity {z=-100}
    simSpaceSettings {maxAngularAcceleration=5, maxAngularVelocity=5, maxLinearAcceleration=1000, maxLinearVelocity=1000, worldAlpha=1}
   }
   RigidBody!(ComponentPose, Alpha, OverrideWorldGravity:Vector="0,0,0", ExternalForce) -> (Pose) {
    bClampLinearTranslationLimitToRefPose=true
    bOverrideWorldGravity=true
    cachedBoundsScale=2
    overridePhysicsAsset=&/F_MED_Islander_dyn_perf
    simulationSpace=BaseBoneSpace
    alphaBoolBlend {blendOption=Linear}
    baseBoneRef {boneName=neck_01}
    componentAppliedLinearAccClamp=<2000 2000 2000>
    overrideWorldGravity {z=-980}
    simSpaceSettings {maxAngularAcceleration=5, maxAngularVelocity=5, maxLinearAcceleration=1000, worldAlpha=1}
   }
   RigidBody(OverrideWorldGravity="0,0,0") {
    bClampLinearTranslationLimitToRefPose=true
    bOverrideWorldGravity=true
    cachedBoundsScale=2
    overridePhysicsAsset=&F_MED_Islander_dyn_perf
    simulationSpace=BaseBoneSpace
    alphaBoolBlend {blendOption=Linear}
    baseBoneRef {boneName=head}
    componentAppliedLinearAccClamp=<2000 2000 2000>
    overrideWorldGravity {z=-980}
    simSpaceSettings {maxAngularAcceleration=5, maxAngularVelocity=5, maxLinearAcceleration=1000, worldAlpha=1}
   }
   n3 = NOT Boolean(A=Get bRigidBodyEnabled())
   n4 = "Constraint - dyn_hair_l_01 (PP)"!(
    ComponentPose=Local To Component(LocalPose=Use cached pose 'prePose'())
    bAlphaBoolEnabled:bool=@n3
    ConstraintWeights_0=0.7
    ConstraintWeights_1:float=0.3
   ) -> (Pose) {
    alphaInputType=Bool
    alphaBoolBlend {blendOption=Linear}
    boneToModify {boneName=dyn_hair_l_01}
    constraintSetup [{targetBone {boneName=spine_05}}, {targetBone {boneName=head}}]
    constraintWeights [1,1]
   }
   n5 = "Constraint - dyn_hair_l_03 (PP)"!(ComponentPose=@n4, bAlphaBoolEnabled:bool=@n3, ConstraintWeights_0=0.8, ConstraintWeights_1:float=0.2) -> (Pose) {
    alphaInputType=Bool
    alphaBoolBlend {blendOption=Linear}
    boneToModify {boneName=dyn_hair_l_03}
    constraintSetup [{targetBone {boneName=spine_05}}, {targetBone {boneName=spine_04}}]
    constraintWeights [1,1]
   }
   n6 = "Constraint - dyn_hair_r_01 (PP)"!(ComponentPose=@n5, bAlphaBoolEnabled:bool=@n3, ConstraintWeights_0=0.7, ConstraintWeights_1:float=0.3) -> (Pose) {
    alphaInputType=Bool
    alphaBoolBlend {blendOption=Linear}
    boneToModify {boneName=dyn_hair_r_01}
    constraintSetup [{targetBone {boneName=spine_05}}, {targetBone {boneName=head}}]
    constraintWeights [1,1]
   }
   n7 = "Constraint - dyn_hair_r_03 (PP)"!(ComponentPose=@n6, bAlphaBoolEnabled:bool=@n3, ConstraintWeights_0=0.8, ConstraintWeights_1:float=0.2) -> (Pose) {
    alphaInputType=Bool
    alphaBoolBlend {blendOption=Linear}
    boneToModify {boneName=dyn_hair_r_03}
    constraintSetup [{targetBone {boneName=spine_05}}, {targetBone {boneName=spine_04}}]
    constraintWeights [1,1]
   }
   n8 = "Transform (Modify) Bone - Bone: dyn_hair_l_04"!(ComponentPose=@n7, bAlphaBoolEnabled:bool=@n3, Translation="0,0,0", Rotation="0,3,0", Scale="1,1,1") -> (Pose) {
    alphaInputType=Bool
    rotationMode=BMM_Additive
    rotationSpace=BCS_BoneSpace
    alphaBoolBlend {blendOption=Linear}
    boneToModify {boneName=dyn_hair_l_04}
    rotation {yaw=3}
   }
   n9 = "Transform (Modify) Bone - Bone: dyn_hair_r_04"!(ComponentPose=@n8, bAlphaBoolEnabled:bool=@n3, Translation="0,0,0", Rotation="0,-5,0", Scale="1,1,1") -> (Pose) {
    alphaInputType=Bool
    rotationMode=BMM_Additive
    rotationSpace=BCS_BoneSpace
    alphaBoolBlend {blendOption=Linear}
    boneToModify {boneName=dyn_hair_r_04}
    rotation {yaw=-5}
   }
   n10 = RigidBody!(ComponentPose=@n9, Alpha, ExternalForce=Get gravity_override() -> (gravity_override:Vector)) -> (Pose) {
    bClampLinearTranslationLimitToRefPose=true
    bOverrideWorldGravity=true
    cachedBoundsScale=2
    overridePhysicsAsset=&F_MED_Islander_dyn_perf
    simulationSpace=BaseBoneSpace
    alphaBoolBlend {blendOption=Linear}
    baseBoneRef {boneName=head}
    componentAppliedLinearAccClamp=<2000 2000 2000>
    overrideWorldGravity {z=-980}
    simSpaceSettings {maxAngularAcceleration=5, maxAngularVelocity=5, maxLinearAcceleration=1000, worldAlpha=1}
   }
   n11 = "Constraint - dyn_hair_l_01 (PP)"(
    ComponentPose=Local To Component(LocalPose=Use cached pose 'prePose'())
    bAlphaBoolEnabled=True
    ConstraintWeights_0=0.7
    ConstraintWeights_1=0.4
   ) {
    alphaInputType=Bool
    alphaBoolBlend {blendOption=Linear}
    boneToModify {boneName=dyn_hair_l_01}
    constraintSetup [{targetBone {boneName=spine_05}}, {targetBone {boneName=head}}]
    constraintWeights [1,1]
   }
   n12 = "Constraint - dyn_hair_l_03 (PP)"(ComponentPose=@n11, bAlphaBoolEnabled=True, ConstraintWeights_0=0.7, ConstraintWeights_1=0.3) {
    alphaInputType=Bool
    alphaBoolBlend {blendOption=Linear}
    boneToModify {boneName=dyn_hair_l_03}
    constraintSetup [{targetBone {boneName=spine_05}}, {targetBone {boneName=spine_04}}]
    constraintWeights [1,1]
   }
   n13 = "Constraint - dyn_hair_r_01 (PP)"(ComponentPose=@n12, bAlphaBoolEnabled=True, ConstraintWeights_0=0.7, ConstraintWeights_1=0.3) {
    alphaInputType=Bool
    alphaBoolBlend {blendOption=Linear}
    boneToModify {boneName=dyn_hair_r_01}
    constraintSetup [{targetBone {boneName=spine_05}}, {targetBone {boneName=head}}]
    constraintWeights [1,1]
   }
   n14 = "Constraint - dyn_hair_r_03 (PP)"(ComponentPose=@n13, bAlphaBoolEnabled=True, ConstraintWeights_0=0.8, ConstraintWeights_1=0.2) {
    alphaInputType=Bool
    alphaBoolBlend {blendOption=Linear}
    boneToModify {boneName=dyn_hair_r_03}
    constraintSetup [{targetBone {boneName=spine_05}}, {targetBone {boneName=spine_04}}]
    constraintWeights [1,1]
   }
   n15 = "Transform (Modify) Bone - Bone: dyn_hair_l_04"(ComponentPose=@n14, bAlphaBoolEnabled=True, Translation="0,0,0", Rotation="0,3,0", Scale="1,1,1") {
    alphaInputType=Bool
    rotationMode=BMM_Additive
    rotationSpace=BCS_BoneSpace
    alphaBoolBlend {blendOption=Linear}
    boneToModify {boneName=dyn_hair_l_04}
    rotation {yaw=3}
   }
   n16 = "Transform (Modify) Bone - Bone: dyn_hair_r_04"(ComponentPose=@n15, bAlphaBoolEnabled=True, Translation="0,0,0", Rotation="0,-5,0", Scale="1,1,1") {
    alphaInputType=Bool
    rotationMode=BMM_Additive
    rotationSpace=BCS_BoneSpace
    alphaBoolBlend {blendOption=Linear}
    boneToModify {boneName=dyn_hair_r_04}
    rotation {yaw=-5}
   }
   n17 = RigidBody(ComponentPose=@n16, ExternalForce=Get gravity_override()) {
    bClampLinearTranslationLimitToRefPose=true
    bOverrideWorldGravity=true
    cachedBoundsScale=2
    overridePhysicsAsset=&/F_MED_Islander_dyn_front_perf
    simulationSpace=BaseBoneSpace
    alphaBoolBlend {blendOption=Linear}
    baseBoneRef {boneName=head}
    componentAppliedLinearAccClamp=<2000 2000 2000>
    overrideWorldGravity {z=-980}
    simSpaceSettings {maxAngularAcceleration=5, maxAngularVelocity=5, maxLinearAcceleration=1000, worldAlpha=1}
   }
   n18 = Blend Poses by bool(BlendPose_0=Component To Local(ComponentPose=@n17), BlendPose_1=Component To Local(ComponentPose=@n10), bActiveValue=Get bIsInFrontEnd()) {blendType=Linear}
   n19 = "Sequence Player (additive) 'F_MED_Islander_Reset_Pose'"() {sequence=&/F_MED_Islander_Reset_Pose}
   n20 = Use cached pose 'prePose'()
   n21 = RigidBody(
    ComponentPose=Local To Component(LocalPose=Apply Additive(Base=@n20, Additive=@n19) {alphaBoolBlend {blendOption=Linear}})
    ExternalForce=Get gravity_override()
   ) {
    bClampLinearTranslationLimitToRefPose=true
    bOverrideWorldGravity=true
    cachedBoundsScale=2
    overridePhysicsAsset=&F_MED_Islander_dyn_perf
    simulationSpace=BaseBoneSpace
    alphaBoolBlend {blendOption=Linear}
    baseBoneRef {boneName=head}
    componentAppliedLinearAccClamp=<2000 2000 2000>
    overrideWorldGravity {z=-980}
    simSpaceSettings {maxAngularAcceleration=5, maxAngularVelocity=5, maxLinearAcceleration=1000, worldAlpha=1}
   }
   n22 = Component To Local(ComponentPose=@n21)
   n23 = Get bIsSkyDiving()
   n24 = Anim Dynamics - Bone: dyn_leaves_01(
    ComponentPose=Local To Component(LocalPose=Blend Poses by bool(BlendPose_0=@n22, BlendPose_1=@n18, bActiveValue=@n23) {blendType=Linear})
    Alpha=1
    ExternalForce="0,0,0"
   ) {
    bAngularSpring=true
    bUseGravityOverride=true
    alphaBoolBlend {blendOption=Linear}
    boundBone {boneName=dyn_leaves_01}
    componentAppliedLinearAccClamp=<1000 1000 1000>
    componentLinearAccScale=<0.2 0.2 0.4>
    componentLinearVelScale=<0.1 0.1 0.2>
    gravityOverride {z=-490}
    physicsBodyDefinitions [{
      sphereCollisionRadius=0
      boundBone {boneName=dyn_leaves_01}
      boxExtents=<6 3 2>
      constraintSetup {angularLimitsMax=<0 10 5>, angularLimitsMin=<0 -10 -5>, angularTarget=<0.25 0 0>}
      localJointOffset=<3 0 0>
     }]
    relativeSpaceBone {boneName=dyn_flowers_MSTR}
   }
   n25 = Anim Dynamics - Bone: dyn_leaves_02(ComponentPose=@n24, Alpha=1, ExternalForce="0,0,0") {
    bAngularSpring=true
    bUseGravityOverride=true
    alphaBoolBlend {blendOption=Linear}
    boundBone {boneName=dyn_leaves_02}
    chainEnd {boneName=dyn_leaves_03}
    componentAppliedLinearAccClamp=<1000 1000 1000>
    componentLinearAccScale=<0.2 0.2 0.4>
    componentLinearVelScale=<0.1 0.1 0.2>
    gravityOverride {z=-490}
    physicsBodyDefinitions [{
      sphereCollisionRadius=0
      boundBone {boneName=dyn_leaves_02}
      boxExtents=<6 8 8>
      constraintSetup {angularLimitsMax=<0 5 5>, angularLimitsMin=<0 -5 -5>, angularTarget=<1 0 0>}
      localJointOffset=<3 0 0>
     }]
    relativeSpaceBone {boneName=dyn_flowers_MSTR}
   }
   n26 = Anim Dynamics - Bone: dyn_leaves_03(ComponentPose=@n25, Alpha=1, ExternalForce="0,0,0") {
    bAngularSpring=true
    bUseGravityOverride=true
    simulationSpace=BoneRelative
    alphaBoolBlend {blendOption=Linear}
    boundBone {boneName=dyn_leaves_03}
    chainEnd {boneName=dyn_leaves_03}
    componentAppliedLinearAccClamp=<1000 1000 1000>
    componentLinearAccScale=<0.2 0.2 0.4>
    componentLinearVelScale=<0.1 0.1 0.2>
    gravityOverride {z=-490}
    physicsBodyDefinitions [{
      sphereCollisionRadius=0
      boundBone {boneName=dyn_leaves_03}
      boxExtents=<6 8 8>
      constraintSetup {angularLimitsMax=<0 5 5>, angularLimitsMin=<0 -5 -5>, angularTarget=<1 0 0>}
      localJointOffset=<3 0 0>
     }]
    relativeSpaceBone {boneName=dyn_leaves_02}
   }
   n27 = Anim Dynamics - Bone: dyn_leaves_04(ComponentPose=@n26, Alpha=1, ExternalForce="0,0,0") {
    bAngularSpring=true
    bUseGravityOverride=true
    alphaBoolBlend {blendOption=Linear}
    boundBone {boneName=dyn_leaves_04}
    componentAppliedLinearAccClamp=<1000 1000 1000>
    componentLinearAccScale=<0.2 0.2 0.4>
    componentLinearVelScale=<0.1 0.1 0.2>
    gravityOverride {z=-490}
    physicsBodyDefinitions [{
      sphereCollisionRadius=0
      boundBone {boneName=dyn_leaves_04}
      boxExtents=<6 3 2>
      constraintSetup {angularLimitsMax=<0 10 5>, angularLimitsMin=<0 -10 -5>, angularTarget=<0.25 0 0>}
      localJointOffset=<3 0 0>
     }]
    relativeSpaceBone {boneName=dyn_flowers_MSTR}
   }
   n28 = Anim Dynamics - Bone: dyn_leaves_05(ComponentPose=@n27, Alpha=1, ExternalForce="0,0,0") {
    bAngularSpring=true
    bUseGravityOverride=true
    alphaBoolBlend {blendOption=Linear}
    boundBone {boneName=dyn_leaves_05}
    componentAppliedLinearAccClamp=<1000 1000 1000>
    componentLinearAccScale=<0.2 0.2 0.4>
    componentLinearVelScale=<0.1 0.1 0.2>
    gravityOverride {z=-490}
    physicsBodyDefinitions [{
      sphereCollisionRadius=0
      boundBone {boneName=dyn_leaves_05}
      boxExtents=<6 3 2>
      constraintSetup {angularLimitsMax=<0 10 5>, angularLimitsMin=<0 -10 -5>, angularTarget=<0.25 0 0>}
      localJointOffset=<3 0 0>
     }]
    relativeSpaceBone {boneName=dyn_flowers_MSTR}
   }
   n29 = Anim Dynamics - Bone: dyn_flowers_01(ComponentPose=@n28, Alpha=1, ExternalForce="0,0,0") {
    bAngularSpring=true
    bUseGravityOverride=true
    alphaBoolBlend {blendOption=Linear}
    boundBone {boneName=dyn_flowers_01}
    chainEnd {boneName=dyn_flowers_02}
    componentAppliedLinearAccClamp=<1000 1000 1000>
    componentLinearAccScale=<0.2 0.2 0.4>
    componentLinearVelScale=<0.1 0.1 0.2>
    gravityOverride {z=-490}
    physicsBodyDefinitions [{
      sphereCollisionRadius=0
      boundBone {boneName=dyn_flowers_01}
      boxExtents=<4 2 2>
      constraintSetup {angularLimitsMax=<0 5 5>, angularLimitsMin=<0 -5 -5>, angularTarget=<1 0 0>}
      localJointOffset=<1.5 0 0>
     }]
    relativeSpaceBone {boneName=dyn_flowers_MSTR}
   }
   n30 = Anim Dynamics - Bone: dyn_flowers_02(ComponentPose=@n29, Alpha=1, ExternalForce="0,0,0") {
    bAngularSpring=true
    bUseGravityOverride=true
    simulationSpace=BoneRelative
    alphaBoolBlend {blendOption=Linear}
    boundBone {boneName=dyn_flowers_02}
    chainEnd {boneName=dyn_flowers_02}
    componentAppliedLinearAccClamp=<1000 1000 1000>
    componentLinearAccScale=<0.2 0.2 0.4>
    componentLinearVelScale=<0.1 0.1 0.2>
    gravityOverride {z=-490}
    physicsBodyDefinitions [{
      sphereCollisionRadius=0
      boundBone {boneName=dyn_flowers_02}
      boxExtents=<4 2 2>
      constraintSetup {angularLimitsMax=<0 5 5>, angularLimitsMin=<0 -5 -5>, angularTarget=<1 0 0>}
      localJointOffset=<1.5 0 0>
     }]
    relativeSpaceBone {boneName=dyn_flowers_01}
   }
   n31 = "Transform (Modify) Bone - Bone: dyn_hair_l_02"!(ComponentPose=@n30, bAlphaBoolEnabled:bool=True, Translation="0,0,0", Rotation="0,5,0", Scale="1,1,1") -> (Pose) {
    alphaInputType=Bool
    rotationMode=BMM_Additive
    rotationSpace=BCS_BoneSpace
    alphaBoolBlend {blendOption=Linear}
    boneToModify {boneName=dyn_hair_l_02}
    rotation {yaw=5}
   }
   Output Pose(Result=Component To Local(ComponentPose=@n31)) {layerGroup=None}
   n32 = "Constraint - dyn_hair_l_MSTR (P)"(ComponentPose=Local To Component(LocalPose=Copy Pose From Mesh() {bUseAttachedParent=true}), Alpha=1, ConstraintWeights_0=1) {
    alphaBoolBlend {blendOption=Linear}
    boneToModify {boneName=dyn_hair_l_MSTR}
    constraintSetup [{targetBone {boneName=neck_02}}]
    constraintWeights [1]
   }
   Save cached pose 'prePose'(Pose=Component To Local(ComponentPose=@n32)) {cacheName=prePose}
   NOT Boolean(A=Get bRigidBodyEnabled())
  ]}
 EventGraph {nodes [
   Event BlueprintUpdateAnimation()
   n1 = BreakTransform(InTransform=GetDeltaTransformFromRefPose(self=GetOwnerSkeletalMeshComponent(), BoneName=head, BaseName=spine_5))
   > Set head_rot_up_r(head_rot_up_r:double=MapRangeClamped(Value=BreakRotator(InRot=@n1.Rotation).Yaw, InRangeA=0, InRangeB=50)) -> (Output_Get:double)
   > n2 = Branch(Condition=Get bIsSwimSprinting() -> (bIsSwimSprinting:bool))
   > Set gravity_override(gravity_override="0,0,5")
   n3 = Branch(execute=@n2.else, Condition=Get bIsSkyDiving())
   > n4 = Branch(Condition=Get bIsParachuteOpened())
   Set gravity_override(execute=[@n3.else, @n4.then], gravity_override="0, 0,-100")
   n5 = Branch(execute=@n4.else, Condition="float > float"(A=Get PawnVelocityZ(), B=50) {is=title K2Node_CallFunction})
   > Set gravity_override(gravity_override="0,10,0")
   Set gravity_override(execute=@n5.else, gravity_override="0,0,400")
   Event OnParachuteOpened()
   > Delay(Duration=1.5)
   > ResetDynamics(InTeleportType=TeleportPhysics)
   > ResetDynamics(InTeleportType=ResetPhysics)
   Delay(Duration=0.05)
   Event OnBeginSkydiving()
   Event OnEndSkydiving()
   Event OnBeginCrouching()
   Event OnEndCrouching()
  ]}
]
variables [
 gravity_override {type=Vector}
 head_rot_up_r {type=number}
]
